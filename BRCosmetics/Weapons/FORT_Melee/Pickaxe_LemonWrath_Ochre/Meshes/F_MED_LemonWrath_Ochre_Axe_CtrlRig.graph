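# Control Rig for the LemonWrath Ochre pickaxe mesh, providing one animation control for the axe's handle bone.
# The forward solve reads the Melee_TwoHanded_Handle_ctrl transform and writes it to the Melee_TwoHanded_Handle bone in local space at full weight.
# The hierarchy is just that bone under b_AR_Root plus an unparented Euler-transform control, with no library functions or other logic.
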ControlRigBlueprint /BRCosmetics/Weapons/FORT_Melee/Pickaxe_LemonWrath_Ochre/Meshes/F_MED_LemonWrath_Ochre_Axe_CtrlRig
graphs [
 RigVMFunctionLibrary {links [], nodes []}
 RigVMModel {nodes [
   RigUnit_BeginExecution = &RigUnit_BeginExecution()
   RigUnit_GetTransform = &RigUnit_GetTransform(Item=(Type=Control,Name="Melee_TwoHanded_Handle_ctrl"), bInitial=False)
   RigUnit_SetTransform = &RigUnit_SetTransform(
    Item=(Type=Bone,Name="Melee_TwoHanded_Handle")
    Space=LocalSpace
    Weight=1
    bInitial=False
    ExecutePin=@RigUnit_BeginExecution.ExecutePin
    Value=@RigUnit_GetTransform.Transform
   )
  ]}
]
hierarchy [
 b_AR_Root {loc=<0 0 0>, rot=<0 0 0 1>}
 Melee_TwoHanded_Handle {parent=b_AR_Root, loc=<0 0 0>, rot=<0 0 0.7071 -0.7071>}
 Melee_TwoHanded_Handle_ctrl {controlType=EULER_TRANSFORM, type=Control, loc=<0 0 0>, rot=<0 0 0 1>}
]
